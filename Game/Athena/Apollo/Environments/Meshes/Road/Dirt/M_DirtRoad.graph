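# Dirt road material for spline meshes that also writes to runtime virtual textures, with a mobile desaturation of base color.
# Blends plain and darkened dirt layers, then adds a flattenable track normal through angle-corrected blending.
# Builds the track mask from a noise-distorted texture with optional center rut, HeightLerp and noise cutout.
# Virtual texture opacity combines a mask with vertex alpha.
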Material /Game/Athena/Apollo/Environments/Meshes/Road/Dirt/M_DirtRoad
graph {
 properties {bUseMaterialAttributes=true, bUsedWithSplineMeshes=true, pixelDepthOffsetMode=PDOM_Legacy, usedShadingModels=Default Lit}
 nodes [
  n1 = TextureCoordinate() {uTiling=2, vTiling=2}
  n2 = TextureSample(UVs=@n1) {samplerSource=SSM_Wrap_WorldGroupSettings, texture=/Game/Environments/Apollo/Sets/Terrain/Textures/T_Apollo_Dirt_D_01}
  n3 = TextureSample(UVs=@n1) {
   samplerSource=SSM_Wrap_WorldGroupSettings
   samplerType=SAMPLERTYPE_Masks
   texture=/Game/Environments/Apollo/Sets/Terrain/Textures/T_Apollo_Dirt_S_01
  }
  n4 = TextureSample(UVs=@n1) {
   samplerSource=SSM_Wrap_WorldGroupSettings
   samplerType=SAMPLERTYPE_Normal
   texture=/Game/Environments/Apollo/Sets/Terrain/Textures/T_Apollo_Dirt_N_01
  }
  n5 = SetMaterialAttributes!(MaterialAttributes, Base Color=@n2.RGB, Specular=@n3.R, Roughness=@n3.R, Normal=@n4.RGB) {attributeSetTypes [
    #1
    #2
    #3
    #4
   ]}
  n6 = SetMaterialAttributes(
   Base Color=Power(Base=@n2.RGB) {constExponent=1.2}
   Specular=@n3.R
   Roughness=@n3.R
   Normal=LinearInterpolate(
    A=@n4.RGB
    B=TextureSample() {samplerType=SAMPLERTYPE_Normal, texture=/Game/Environments/Apollo/Sets/Terrain/Textures/T_Apollo_Dirt_N_02}.RGB
    Alpha=NamedRerouteUsage()
   )
  ) {attributeSetTypes [
    #1
    #2
    #3
    #4
   ]}
  n7 = GetMaterialAttributes(None=BlendMaterialAttributes(A=@n5, B=@n6, Alpha=NamedRerouteUsage())) {attributeGetTypes [#4]}
  n8 = TextureSampleParameter2D() {parameterName=TrackNormal, samplerType=SAMPLERTYPE_Normal, texture=&/T_DirtRoad01_N}
  n9 = MaterialFunctionCall(Normal=@n8.RGB, Flatness=OneMinus(None=NamedRerouteUsage())) {materialFunction=&FlattenNormal}
  NamedRerouteDeclaration(None=SetMaterialAttributes!(
    MaterialAttributes=@n7.MaterialAttributes
    Normal=MaterialFunctionCall(BaseNormal=@n7.Normal, AdditionalNormal=@n9.Result) {materialFunction=&BlendAngleCorrectedNormals}.Result
   ) {attributeSetTypes [#4]}) {name=finalMaterial, nodeColor=<rgba 0.4766 0 1 1>}
  n10 = TextureSample() {samplerType=SAMPLERTYPE_LinearGrayscale, texture=&/T_DirtRoad01_M2}
  n11 = ScalarParameter() {defaultValue=2, parameterName=RVTBlendMaskBoost}
  NamedRerouteDeclaration(None=Saturate(None=Multiply(A=@n10.R, B=@n11))) {name=RVTblendMask, nodeColor=<rgba 0.4297 1 0 1>}
  n12 = Divide(A=WorldPosition().XYZ, B=ScalarParameter() {defaultValue=4788.1494, parameterName=WorldAlignedDistortionSize}) {constB=1000}
  NamedRerouteDeclaration(None=TextureSample(UVs=ComponentMask(None=@n12) {g=true, r=true}) {texture=&T_Apollo_Dirt_D_01}.A) {name=WorldAlignedDistortion, nodeColor=<rgba 1 0.4922 0 1>}
  n13 = Add(A=NamedRerouteUsage()) {constB=-0.5}
  n14 = ScalarParameter() {defaultValue=0.4442, parameterName=Distortion}
  n15 = TextureSampleParameter2D(UVs=Add(A=TextureCoordinate(), B=Multiply(A=@n13, B=@n14) {constB=0.2})) {parameterName=TrackMask, samplerType=SAMPLERTYPE_LinearGrayscale, texture=&/T_DirtRoad01_M}
  n16 = ScalarParameter() {parameterName=useCenterRut}
  n17 = MaterialFunctionCall(
   Transition Phase=Multiply(A=@n15.A, B=LinearInterpolate(B=@n10.R, Alpha=@n16) {constA=1})
   Height Texture=NamedRerouteUsage()
   Contrast=ScalarParameter() {defaultValue=0.2, parameterName=TrackMaskContrast}
  ) {materialFunction=&HeightLerp}
  n18 = NamedRerouteUsage()
  n19 = ScalarParameter() {defaultValue=2, parameterName=WorldAlignedNoiseAlphaCutoutFalloff}
  n20 = Saturate(None=Multiply(A=Power(Base=@n18, Exp=@n19), B=ScalarParameter() {defaultValue=2, parameterName=WorldAlignedNoiseAlphaCutout}))
  NamedRerouteDeclaration(None=Saturate(None=Subtract(A=@n17.Alpha, B=@n20))) {name=TrackMask, nodeColor=<rgba 1 0 0.8672 1>}
  NamedRerouteUsage()
  n21 = GetMaterialAttributes(None=NamedRerouteUsage()) {attributeGetTypes [
    #1
    #2
    #3
    #4
   ]}
  n22 = ScalarParameter() {defaultValue=0.15, parameterName=Mobile Desaturation}
  RuntimeVirtualTextureOutput!(
   BaseColor=ShadingPathSwitch(Default=@n21.Base Color, Mobile=Desaturation(None=@n21.Base Color, Fraction=@n22))
   Specular=@n21.Specular
   Roughness=@n21.Specular
   Normal=@n21.Specular
   WorldHeight
   Opacity=Multiply(A=NamedRerouteUsage(), B=VertexColor().A)
   Mask
   Displacement
   Mask4
  )
 ]
}
